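# Toon character body material for the F_MED_Dazzle Season 24 Super Level style, using material attributes and the Cloth shading model on skeletal meshes.
# A shared PreFX function feeds anime shading, cloth-sim WPO and anime SkinFX in sequence.
# The Season 24 Super Level function, borrowed from the M_MED_LocalZilla folder, then combines toon shading, inkline and rerouted PreFX masks and pre-skinned position before the shared PostFX function finishes the chain.
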Material /BRCosmetics/Characters/Player/Female/Medium/Bodies/F_MED_Dazzle/Material/Super_Levels/F_S24_Toon_SL
graph {
 properties {
  bAutomaticallySetUsageInEditor=false
  bUseMaterialAttributes=true
  bUsedWithMeshDeformer=true
  bUsedWithMorphTargets=true
  bUsedWithSkeletalMesh=true
  pixelDepthOffsetMode=PDOM_Legacy
  shadingModel=MSM_Cloth
  usedShadingModels=Cloth
 }
 nodes [
  n1 = MaterialFunctionCall() {materialFunction=&BaseFN_CharacterFunction_PreFX}
  n2 = MaterialFunctionCall!(
   Material=@n1.Result
   LitDiffuseOverride
   SSC_AlphaOverride
   Tint Mask
   DF InkLine Visibility Mask
   DF Inkline Injection
   DF Inkline Color Override
   DF Inkline Color Override Mask
   CustomUVs
  ) {materialFunction=/Game/Packages/Fortress_SharedMaterials/Anime_Master_Material/MF_AnimeShading}
  n3 = MaterialFunctionCall!(Material=@n2.Result, SkinFX_Mask=@n2.FX) {materialFunction=/Game/Global/MaterialFunctions/MF_ClothSimWPO_v2}
  MaterialFunctionCall(
   AO=@n1.AO
   Distance Blend=@n1.AO
   Fresnel=@n1.AO
   Gmap=@n1.AO
   Pre Skinned Local Position=@n1.AO
   Material=MaterialFunctionCall!(
    In=MaterialFunctionCall!(MA=@n3.Result, Shading=@n2.Shading, InkLine Mask=@n2.InkLine Mask, SSC=@n2.SSC, SkinFX=@n2.FX) {materialFunction=/Game/Packages/Fortress_SharedMaterials/Anime_Master_Material/MF_AnimeSkinFX}.MA
    Preskinned Position=NamedRerouteUsage()
    AO=NamedRerouteUsage()
    Gmap=NamedRerouteUsage()
    Skin=NamedRerouteUsage()
    Toon Shading=@n2.Shading
    Inkline=@n2.InkLine Mask
   ) {materialFunction=/Game/Characters/Player/Male/Medium/Bodies/M_MED_LocalZilla/Materials/Super_Levels/SL24_Functions/MFC_S24_SuperLevel}.Result
  ) {materialFunction=&BaseFN_CharacterFunction_PostFX}
  NamedRerouteDeclaration(None=@n1.AO) {name=AO, nodeColor=<rgba 0.5 0 1 1>}
  NamedRerouteDeclaration(None=@n1.Gmap) {name=Gmap, nodeColor=<rgba 0.4531 0 1 1>}
  NamedRerouteDeclaration(None=@n1.Skin) {name=Skin, nodeColor=<rgba 0 1 0.3203 1>}
  NamedRerouteDeclaration(None=@n1.Pre skinned Local Position) {name=Preskinned Position, nodeColor=<rgba 0.4766 0 1 1>}
 ]
}
